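# Sound cue for the SkirmishFemale back bling's kick sound, set up as reactive cosmetic audio.
# A single kick wave plays through a Delay node that holds it back by a fixed one second.
# It uses the back bling sound class, a close-range pet attenuation and a reactive back bling property sheet.
# Concurrency is overridden to one instance per owner, blocking new plays while one is active.
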SoundCue /BRCosmetics/Sounds/Backblings/SkirmishFemale/BackBling_SkirmishFemale_Kick_Cue
root=Delay_0
nodes [
 {id=Delay_0, children [WavePlayer_0], properties {delayMax=1, delayMin=1}}
 {id=WavePlayer_0, properties {soundWaveAssetPtr=&/BackBling_SkirmishFemale_Kick_01}}
]
properties {
 attenuationSettings=&Fort_Pets_Close_Athena_Att
 audioPropertiesSheet=/BRCosmetics/Sounds/Settings/PropertySheets/Reactive/APS_Cosmetic_Reactive_Backbling
 bOverrideConcurrency=true
 maxDistance=1800.0001
 soundClassObject=/Game/Sounds/FortSoundClassesAndModes/Fort_Backblings
 volumeMultiplier=1.2
 concurrencyOverrides {bLimitToOwner=true, maxCount=1, resolutionRule=PreventNew}
}
